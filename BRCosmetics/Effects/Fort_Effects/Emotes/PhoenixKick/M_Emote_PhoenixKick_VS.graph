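# Translucent, unlit, two-sided Niagara sprite and mesh particle material for the Phoenix Kick emote VS effect.
# Dynamic parameters scroll the base texture in U and move a vertical appear mask.
# Switch-style scalar lerps select color and alpha sources, or replace them with a rotated, panning sweep texture inside a VS mask.
# Particle color tints emissive through MF_FX_EmissiveAdjustments; opacity gets view-angle and depth fades.
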Material /BRCosmetics/Effects/Fort_Effects/Emotes/PhoenixKick/M_Emote_PhoenixKick_VS
graph {
 properties {
  bUsedWithNiagaraMeshParticles=true
  bUsedWithNiagaraSprites=true
  blendMode=BLEND_Translucent
  pixelDepthOffsetMode=PDOM_Legacy
  shadingModel=MSM_Unlit
  twoSided=true
  usedShadingModels=Unlit
 }
 nodes [
  n1 = ParticleColor()
  n2 = DynamicParameter() {defaultValue=<rgba 1 0 1 1>, paramNames [BaseTexture_AddU, Appear, Param3, Param4]}
  n3 = Constant()
  n4 = TextureSampleParameter2D(UVs=Add(A=TextureCoordinate(), B=AppendVector(A=@n2.BaseTexture_AddU, B=@n3))) {parameterName=BaseTexture, texture=&/T_Emote_PhoenixKick_VS_Blur}
  n5 = LinearInterpolate(
   A=@n4.RGB
   B=LinearInterpolate(A=@n4.R, B=@n4.R, Alpha=ScalarParameter() {parameterName=Use_VSChange})
   Alpha=ScalarParameter() {parameterName=Use_BaseRGB}
  )
  n6 = ScalarParameter() {parameterName=Use_BaseTextureAlpha}
  n7 = TextureSample() {texture=&/T_Emote_PhoenixKick_VSMask_RGB}
  n8 = ScalarParameter() {parameterName=Use_Sweep}
  n9 = LinearInterpolate(A=Multiply(A=@n1.A, B=LinearInterpolate(A=@n5, B=@n4.A, Alpha=@n6)), B=@n7.R, Alpha=@n8)
  n10 = Constant()
  n11 = DynamicParameter() {defaultValue=<rgba 1 0 1 1>, paramNames [BaseTexture_AddU, Appear, Param3, Param4]}
  n12 = ComponentMask(None=Add(A=TextureCoordinate(), B=AppendVector(A=@n10, B=@n11.Appear))) {g=true}
  n13 = Power(Base=Multiply(A=OneMinus(None=@n12)) {constB=5}) {constExponent=5}
  n14 = LinearInterpolate(A=@n9, B=Multiply(A=@n9, B=Saturate(None=@n13)), Alpha=ScalarParameter() {parameterName=Use_AppearMask})
  n15 = ScalarParameter() {defaultValue=1, parameterName=Opacity Mult}
  n16 = ScalarParameter() {defaultValue=1, group=Alpha, parameterName=FadeViewAngle}
  DepthFade(
   Opacity=Multiply(
    A=Multiply(A=@n14, B=@n15)
    B=MaterialFunctionCall("FadeSensitivity (optional)"=@n16) {materialFunction=&FadeBasedOnViewAngle}.Result
   )
   FadeDistance=ScalarParameter() {defaultValue=10, group=Alpha, parameterName=FadeDistance}
  )
  n17 = VectorParameter() {parameterName=Color A, defaultValue=<rgba 0 0.0833 1 1>}
  n18 = VectorParameter() {parameterName=Color B, defaultValue=<rgba 0 0.1667 1 1>}
  n19 = Panner(Coordinate=MaterialFunctionCall(UVs=TextureCoordinate(), "Rotation Angle (0-1)"=Constant() {r=-0.08}) {materialFunction=&CustomRotator}.Rotated Values) {speedY=0.2}
  n20 = LinearInterpolate(
   A=LinearInterpolate(A=LinearInterpolate(A=@n17.RGB, B=@n18.RGB), B=@n5, Alpha=ScalarParameter() {parameterName=Use_DefaultBaseTextureColor})
   B=Multiply(A=TextureSample(UVs=@n19) {texture=&/T_Emote_PhoenixKick_Sweep}.RGB, B=@n7.R)
   Alpha=@n8
  )
  n21 = ScalarParameter() {defaultValue=3, parameterName=Emissive Mult}
  MaterialFunctionCall(Emissive=Multiply(A=Multiply(A=@n20, B=@n21), B=@n1.RGB)) {materialFunction=&MF_FX_EmissiveAdjustments}
 ]
}
