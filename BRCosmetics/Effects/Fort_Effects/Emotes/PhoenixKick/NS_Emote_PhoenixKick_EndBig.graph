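# Niagara system for the large end burst of the Phoenix Kick emote, fired about 180 units above its origin.
# Short-lived sprite emitters layer glow, flare, ping and RGB flashes with outward-flying spikes and black streaks, plus randomly oriented GPU mesh rays in orange tones.
# Curves drive size, color and fade, time-of-day scalars adjust several layers' brightness, and extra per-particle color and size attributes support multiple renderers per particle.
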NiagaraSystem /BRCosmetics/Effects/Fort_Effects/Emotes/PhoenixKick/NS_Emote_PhoenixKick_EndBig
dependencies {
 usedDataInterfaces [&NiagaraDataInterfaceCurve, &NiagaraDataInterfaceColorCurve, &NiagaraDataInterfaceVector2DCurve]
 usedDynamicInputs []
 usedModules [&SystemState]
}
emitters {
 Burst_glow {
  emitterUpdateScript [
   &EmitterState(Inactive Response=::NewEnumerator1, Life Cycle Mode=::NewEnumerator1, Loop Behavior=::NewEnumerator1, Loop Duration=2, UseLoopDelay=true)
   &SpawnBurst_Instantaneous(Use Spawn Probability=false, UseLoopCountLimit=false)
  ]
  particleSpawnScript [&InitializeParticle(
    Color=<rgba 0.2622 1 0.8395 1>
    Lifetime=1.2
    Mesh Scale Mode=::NewEnumerator3
    Position Offset=<0 0 185>
    Sprite Size Mode=::NewEnumerator3
    Uniform Sprite Size=200
    UsePositionOffset=true
    Write Material Random=true
   )]
  particleUpdateScript [
   &ParticleState()
   /SetVariables(Particles.Color003:LinearColor=&MakeLinearColorFromVectorAndFloat(
     "Vector (RGB)"=&MakeVectorFromLinearColorRGB(Color=&ColorFromCurve(ColorCurve=DataInterface() {propertyValues {
         AlphaCurve {keys []}
         BlueCurve {keys [<at 0.1407 0.1415 RCIM_Cubic>,<at 0.356 0.0128 RCIM_Cubic>]}
         GreenCurve {keys [<at 0.1407 2.0576 RCIM_Cubic>,<at 0.356 0.2345 RCIM_Cubic>]}
         RedCurve {keys [<at 0.1407 5 RCIM_Cubic>,<at 0.356 1 RCIM_Cubic>]}
        }}))
     "Float (Alpha)"=&FloatFromCurve(
      FloatCurve=DataInterface() {propertyValues {Curve {keys [
          <at 0 0>
          {arriveTangent=-3.2815, interpMode=RCIM_Cubic, leaveTangent=-3.2815, tangentMode=RCTM_User, time=0.1, value=1}
          {arriveTangent=-2.3069, interpMode=RCIM_Cubic, leaveTangent=-2.3069, tangentMode=RCTM_User, time=0.3, value=0.4}
          <at 0.6 0 RCIM_Cubic>
         ]}}}
      Scale Curve=&Multiply_Float(
       A=1
       B=&Multiply_Float(A=&FNTimeOfDayScalar(Night Scale=0.8, Morning Scale=1, Evening Scale=1), B=1)
      )
     )
    )) {isSetParameters=true}
   /SetVariables(Particles.Size003:Vector2f=&Multiply_Vector2D(
     A=&Multiply_Vector2D_ByFloat(Float=2, Vector2D=<400 180>)
     B=&MakeVector2D(
      X=&FloatFromCurve(FloatCurve=DataInterface() {propertyValues {Curve {keys [
           {arriveTangent=1.7907, interpMode=RCIM_Cubic, leaveTangent=1.7907, tangentMode=RCTM_User, time=0, value=0.3}
           {arriveTangent=1.1, interpMode=RCIM_Cubic, leaveTangent=1.1, time=0.3, value=1}
           <at 1 1.4 RCIM_Cubic>
          ]}}})
      Y=&FloatFromCurve(FloatCurve=DataInterface() {propertyValues {Curve {keys [{arriveTangent=-1.906, interpMode=RCIM_Cubic, leaveTangent=-1.906, tangentMode=RCTM_User, time=0, value=1}, <at 1 0 RCIM_Cubic>]}}})
     )
    )) {isSetParameters=true}
   /SetVariables(Particles.Color002=&MakeLinearColorFromVectorAndFloat(
     "Vector (RGB)"=&MakeVectorFromLinearColorRGB(Color=<rgba 0.0365 0.0053 0 1>)
     "Float (Alpha)"=&FloatFromCurve(FloatCurve=DataInterface() {propertyValues {Curve {keys [<at 0 0 RCIM_Cubic>,<at 0.1579 0.9974 RCIM_Cubic>,<at 0.5 0 RCIM_Cubic>]}}})
    )) {isSetParameters=true}
   /SetVariables(Particles.Size002:Vector2f=&Vector2DFromFloat(Value=&FloatFromCurve(
      FloatCurve=DataInterface() {propertyValues {Curve {keys [
          <at 0 0.55 RCIM_Cubic>
          {arriveTangent=-0.001, interpMode=RCIM_Cubic, leaveTangent=-0.001, tangentMode=RCTM_User, time=0.2, value=1}
          <at 1 0 RCIM_Cubic>
         ]}}}
      Scale Curve=300
     ))) {isSetParameters=true}
  ]
  rendererClasses [&NiagaraSpriteRendererProperties]
  renderers [
   {rendererClass=&NiagaraSpriteRendererProperties}
   {rendererClass=&NiagaraSpriteRendererProperties, rendererIndex=1}
  ]
 }
 DiscoBallRaySharp001 {
  simTarget=GPUComputeSim
  emitterUpdateScript [
   &EmitterState(Life Cycle Mode=::NewEnumerator1, Loop Behavior=::NewEnumerator1, Loop Duration=1, UseLoopDelay=false)
   &SpawnRate(SpawnRate=15, Use Spawn Probability=false) {enabled=false}
   &SpawnBurst_Instantaneous(Spawn Count=10, Use Spawn Probability=false)
  ]
  particleSpawnScript [
   &InitializeParticle(
    Color=&RandomRangeLinearColor(Minimum=<rgba 2 0.0723 0 3>, Maximum=<rgba 1 0.6223 0.0043 3>)
    Lifetime Max=0.5
    Lifetime Min=0.3
    Lifetime Mode=::NewEnumerator1
    Mesh Scale Max=<1.2 0.3 0.3>
    Mesh Scale Min=<1.5 0.2 0.2>
    Mesh Scale Mode=::NewEnumerator2
    Position Offset=<0 0 185>
    UsePositionOffset=true
    Write Material Random=true
   ) {module=InitializeParticle001}
   &DynamicMaterialParameters(
    Index 0 Param 1=&UniformRangedFloat(Minimum=-1, Maximum=1)
    Index 0 Param 2=1
    Module.Param0WriteEnabled=true
    Module.Param1WriteEnabled=true
    Module.Param2WriteEnabled=true
    Module.Param3WriteEnabled=true
   )
   &InitialMeshOrientation(
    Orientation Vector=<1 0 1>
    Rotation=&Add_Vector(A=<0 0 0>, B=&RandomRangeVector(Minimum=<-1 0 -1>, Maximum=<1 1 1>))
    Use Orientation Vector=true
    Use Rotation Vector=true
   ) {module=InitialMeshOrientation001}
  ]
  particleUpdateScript [
   &CameraOffset(Camera Offset Amount=-30) {enabled=false}
   &ParticleState()
   &SolveForcesAndVelocity(Clamp Velocity=false, Limit Acceleration=false)
   &ScaleColor(
    Scale Alpha=&Multiply_Float(
     A=&FloatFromCurve(
      FloatCurve=DataInterface() {propertyValues {Curve {keys [
          {interpMode=RCIM_Cubic, tangentMode=RCTM_User, time=0, value=0}
          {arriveTangent=-1.6906, interpMode=RCIM_Cubic, leaveTangent=-1.6906, tangentMode=RCTM_User, time=0.1, value=0.5}
          {tangentMode=RCTM_User, time=1, value=0}
         ]}}}
      Scale Curve=&FloatFromCurve(FloatCurve=DataInterface() {propertyValues {Curve {keys [<at 0 0>,<at 0.3005 1>,<at 1 1>]}}})
     )
     B=&RandomRangeFloat(Minimum=1, Maximum=2.2)
    )
    Scale RGB=&Multiply_VectorByFloat(
     Vector=<1 1 1>
     Float=&FNTimeOfDayScalar(Night Scale=0.5, Morning Scale=0.3, Day Scale=0.7, Evening Scale=0.4)
    )
    ScaleA=true
    ScaleRGB=true
   )
   &ScaleMeshSize(Scale Factor=&MakeVector(
     X=&FloatFromCurve(FloatCurve=DataInterface() {propertyValues {Curve {keys [<at 0 0.8>,<at 1 1.2>]}}}, Scale Curve=1.65)
     Y=1
     Z=&FloatFromCurve(FloatCurve=DataInterface() {propertyValues {Curve {keys [<at 0 1.8>,<at 1 1>]}}}, Scale Curve=1.65)
    ))
  ]
  rendererClasses [&NiagaraMeshRendererProperties]
  renderers [{rendererClass=&NiagaraMeshRendererProperties, rendererIndex=0}]
 }
 Flare {
  emitterUpdateScript [
   &EmitterState(Inactive Response=::NewEnumerator1, Life Cycle Mode=::NewEnumerator1, Loop Behavior=::NewEnumerator1, Loop Duration=2, UseLoopDelay=true)
   &SpawnBurst_Instantaneous(Use Spawn Probability=false, UseLoopCountLimit=false)
  ]
  particleSpawnScript [&InitializeParticle(
    Color=<rgba 1 0.5 0 1>
    Lifetime=0.3
    Position Offset=<0 0 185>
    Sprite Size Mode=::NewEnumerator3
    Uniform Sprite Size=170
    UsePositionOffset=true
    Write Material Random=true
   )]
  particleUpdateScript [
   &ParticleState()
   &ScaleColor(
    Scale Alpha=&FloatFromCurve(
     FloatCurve=DataInterface() {propertyValues {Curve {keys [
         {arriveTangent=0.0553, interpMode=RCIM_Cubic, leaveTangent=0.0553, tangentMode=RCTM_User, time=0, value=1}
         {arriveTangent=-1.6036, interpMode=RCIM_Cubic, leaveTangent=-1.6036, tangentMode=RCTM_User, time=0.5, value=0.4}
         <at 1 0 RCIM_Cubic>
        ]}}}
     Scale Curve=5
    )
    Scale RGB=&Multiply_VectorByFloat(
     Vector=&VectorFromFloat(Value=&FNTimeOfDayScalar(Night Scale=0.5, Morning Scale=0.25, Evening Scale=0.5))
     Float=&FN_FrontEndFloatModifier(Gameplay=1, Front End=1)
    )
    ScaleA=true
    ScaleRGB=true
   )
   &ScaleSpriteSize(
    Uniform Curve Scale=2
    Uniform Curve Sprite Scale=DataInterface() {propertyValues {Curve {keys [{arriveTangent=1.5627, interpMode=RCIM_Cubic, leaveTangent=1.5627, tangentMode=RCTM_User, time=0, value=0.45}, <at 1 1.4 RCIM_Cubic>]}}}
   )
   &DynamicMaterialParameters(
    Index 0 Param 2=2
    Module.Param0WriteEnabled=true
    Module.Param1WriteEnabled=true
    Module.Param2WriteEnabled=true
    Module.Param3WriteEnabled=true
   )
   /SetVariables(Particles.Color002=&MakeLinearColorFromVectorAndFloat("Vector (RGB)"=&MakeVectorFromLinearColorRGB(Color=<rgba 1 0.5 0 1>), "Float (Alpha)"=0.5)) {isSetParameters=true}
   /SetVariables(Particles.Size002:Vector2f=&Multiply_Vector2D(A=$Particles.SpriteSize, B=<0.8 0.8>)) {isSetParameters=true}
   /SetVariables(Particles.Dyn002:Vector4f=&MakeVector4(
     X=&FloatFromCurve(FloatCurve=DataInterface() {propertyValues {Curve {keys [<at 0 0.5>,<at 1 0>]}}})
     Y=0
     Z=0
     W=0
    )) {isSetParameters=true}
  ]
  rendererClasses [&NiagaraSpriteRendererProperties]
  renderers [
   {rendererClass=&NiagaraSpriteRendererProperties}
   {rendererClass=&NiagaraSpriteRendererProperties, rendererIndex=1}
  ]
 }
 RGB {
  emitterUpdateScript [
   &EmitterState(Inactive Response=::NewEnumerator1, Life Cycle Mode=::NewEnumerator1, Loop Behavior=::NewEnumerator1, Loop Duration=2, UseLoopDelay=true)
   &SpawnBurst_Instantaneous(Use Spawn Probability=false, UseLoopCountLimit=false)
  ]
  particleSpawnScript [&InitializeParticle(
    Color Mode=::NewEnumerator0
    Lifetime=0.35
    Sprite Size Mode=::NewEnumerator3
    Uniform Sprite Size=130
    UsePositionOffset=false
    Write Material Random=true
   )]
  particleUpdateScript [
   &ParticleState()
   &ScaleSpriteSize(
    Uniform Curve Scale=1.5
    Uniform Curve Sprite Scale=DataInterface() {propertyValues {Curve {keys [
        {arriveTangent=0.7997, interpMode=RCIM_Cubic, leaveTangent=0.7997, tangentMode=RCTM_User, time=0, value=0.9}
        {arriveTangent=0.5, interpMode=RCIM_Cubic, leaveTangent=0.5, time=0.3, value=1.15}
        {arriveTangent=0.1863, interpMode=RCIM_Cubic, leaveTangent=0.1863, tangentMode=RCTM_User, time=1, value=1.4}
       ]}}}
   )
   &ShapeLocation(MaskSpawnGroup=false, Offset=<0 0 178>, Offset Mode=::NewEnumerator0, Override Seed=false, Sphere Radius=0.1, UseOwnerScale=true)
   &ScaleColor(
    Scale Alpha=&FloatFromCurve(
     FloatCurve=DataInterface() {propertyValues {Curve {keys [
         {arriveTangent=-3.2378, interpMode=RCIM_Cubic, leaveTangent=-3.2378, tangentMode=RCTM_User, time=0, value=1}
         {arriveTangent=-0.5463, interpMode=RCIM_Cubic, leaveTangent=-0.5463, tangentMode=RCTM_User, time=0.3, value=0.2}
         <at 1 0 RCIM_Cubic>
        ]}}}
     Scale Curve=5
    )
    Scale RGB=<1.5 1.5 1.5>
    ScaleA=true
    ScaleRGB=true
   )
  ]
  rendererClasses [&NiagaraSpriteRendererProperties]
  renderers [
   {rendererClass=&NiagaraSpriteRendererProperties}
   {rendererClass=&NiagaraSpriteRendererProperties, rendererIndex=1}
   {rendererClass=&NiagaraSpriteRendererProperties, rendererIndex=2}
  ]
 }
 RGB001 {
  emitterUpdateScript [
   &EmitterState(
    Inactive Response=::NewEnumerator1
    Life Cycle Mode=::NewEnumerator1
    Loop Behavior=::NewEnumerator1
    Loop Delay=0.05
    Loop Duration=10
    UseLoopDelay=true
   )
   &SpawnBurst_Instantaneous(Use Spawn Probability=false, UseLoopCountLimit=false)
   &SpawnBurst_Instantaneous(Spawn Time=0.15, Use Spawn Probability=false, UseLoopCountLimit=false) {module=SpawnBurst_Instantaneous001}
  ]
  particleSpawnScript [
   &InitializeParticle(
    Color Mode=::NewEnumerator0
    Lifetime=0.3
    Sprite Size Mode=::NewEnumerator3
    Uniform Sprite Size=170
    UsePositionOffset=false
    Write Material Random=true
   )
   &CameraOffset(Camera Offset Amount=10) {enabled=false}
  ]
  particleUpdateScript [
   &ParticleState()
   &ScaleSpriteSize(
    Uniform Curve Scale=2
    Uniform Curve Sprite Scale=DataInterface() {propertyValues {Curve {keys [
        {arriveTangent=1.2935, interpMode=RCIM_Cubic, leaveTangent=1.2935, tangentMode=RCTM_User, time=0, value=0.8}
        {arriveTangent=0.6116, interpMode=RCIM_Cubic, leaveTangent=0.6116, tangentMode=RCTM_User, time=1, value=1.9}
       ]}}}
   )
   &ShapeLocation(MaskSpawnGroup=false, Offset=<0 0 178>, Offset Mode=::NewEnumerator0, Override Seed=false, Sphere Radius=0.1, UseOwnerScale=true)
   &ScaleColor(
    Scale Alpha=&FloatFromCurve(FloatCurve=DataInterface() {propertyValues {Curve {keys [
         {arriveTangent=-4.498, interpMode=RCIM_Cubic, leaveTangent=-4.498, tangentMode=RCTM_User, time=0, value=1}
         {arriveTangent=-0.6247, interpMode=RCIM_Cubic, leaveTangent=-0.6247, tangentMode=RCTM_User, time=0.25, value=0.15}
         <at 1 0 RCIM_Cubic>
        ]}}})
    ScaleA=true
    ScaleRGB=true
   )
  ]
  rendererClasses [&NiagaraSpriteRendererProperties]
  renderers [
   {rendererClass=&NiagaraSpriteRendererProperties}
   {rendererClass=&NiagaraSpriteRendererProperties, rendererIndex=1}
   {rendererClass=&NiagaraSpriteRendererProperties, rendererIndex=2}
  ]
 }
 Spikes {
  emitterUpdateScript [
   &EmitterState(Life Cycle Mode=::NewEnumerator1, Loop Behavior=::NewEnumerator1, Loop Duration=1, UseLoopDelay=false)
   &SpawnBurst_Instantaneous(Spawn Count=8, Use Spawn Probability=false, UseLoopCountLimit=false)
  ]
  particleSpawnScript [
   &InitializeParticle(
    Lifetime Max=0.5
    Lifetime Min=0.2
    Lifetime Mode=::NewEnumerator1
    Sprite Rotation Mode=::NewEnumerator2
    Sprite Size=<10 50>
    Sprite Size Mode=::NewEnumerator4
    UsePositionOffset=false
    Write Material Random=true
   )
   &ShapeLocation(MaskSpawnGroup=false, Offset=<0 0 185>, Offset Mode=::NewEnumerator0, Override Seed=false, Sphere Radius=10, UseOwnerScale=true)
   &AddVelocity(
    Module.Use Falloff Distance=false
    Origin Offset=<0 0 185>
    Override Seed=false
    Velocity Mode=::NewEnumerator1
    Velocity Speed=&RandomRangeFloat(Minimum=150, Maximum=150)
   )
  ]
  particleUpdateScript [
   &ParticleState()
   &ScaleColor(
    Scale Alpha=20
    Scale RGB=&MakeVectorFromLinearColorRGB(Color=&ColorFromCurve(ColorCurve=DataInterface() {propertyValues {
        AlphaCurve {keys []}
        BlueCurve {keys [
          <at 0.0308 1>
          <at 0.0728 0>
          {arriveTangent=3.7632, interpMode=RCIM_Cubic, leaveTangent=3.7632, time=0.1613, value=0}
          <at 0.1973 0.4688>
          <at 0.3609 0.4114>
          {arriveTangent=-1.2991, interpMode=RCIM_Cubic, leaveTangent=-1.2991, time=0.5485, value=0.0374}
          <at 0.6776 0>
          <at 0.7136 1>
          <at 0.8382 0.4886>
          <at 0.8742 0>
          <at 0.9462 0>
         ]}
        GreenCurve {keys [
          <at 0.0308 1>
          <at 0.0728 0>
          {arriveTangent=6.7108, interpMode=RCIM_Cubic, leaveTangent=6.7108, time=0.1613, value=0}
          <at 0.1973 0.8359>
          <at 0.3609 0.7336>
          {arriveTangent=-1.9884, interpMode=RCIM_Cubic, leaveTangent=-1.9884, time=0.5485, value=0.5964}
          <at 0.6776 0.104>
          <at 0.7136 0.0319>
          <at 0.8382 0.0027>
          <at 0.8742 0.5>
          <at 0.9462 0.5>
         ]}
        RedCurve {keys [
          <at 0.0308 1>
          <at 0.0728 0>
          {arriveTangent=40.141, interpMode=RCIM_Cubic, leaveTangent=40.141, time=0.1613, value=0}
          <at 0.1973 5>
          <at 0.3609 5>
          {arriveTangent=-12.632, interpMode=RCIM_Cubic, leaveTangent=-12.632, time=0.5485, value=1}
          <at 0.6776 1>
          <at 0.7136 0.7977>
          <at 0.8382 1>
          <at 0.8742 1>
          <at 0.9462 1>
         ]}
       }}))
    ScaleA=true
    ScaleRGB=true
   )
   &ScaleSpriteSize(
    Uniform Curve Scale=&RandomRangeFloat(Minimum=3.5, Maximum=4.5)
    Uniform Curve Sprite Scale=DataInterface() {propertyValues {Curve {keys [{tangentMode=RCTM_User, time=0, value=0.6}, <at 0.3 0.85>, {tangentMode=RCTM_User, time=1, value=1}]}}}
   )
   &SpriteRotationRate(Rotation Rate=&RandomRangeFloat(Minimum=-200, Maximum=200))
   &FN_Sprite_Attributes(
    Color=&MakeLinearColorFromVectorAndFloat(
     "Vector (RGB)"=&Multiply_VectorByFloat(Vector=&MakeVectorFromLinearColorRGB(Color=<rgba 1 0.6384 0.1512 1>), Float=1)
     "Float (Alpha)"=&Multiply_Float(A=3, B=1)
    )
    Create Color Attribute=true
    Create SpriteSize Attribute=true
    SpriteSize=&Multiply_Vector2D_ByFloat(
     Float=35
     Vector2D=&Multiply_Vector2D_ByFloat(
      Float=&FloatFromCurve(FloatCurve=DataInterface() {propertyValues {Curve {keys [
           <at 0 1 RCIM_Cubic>
           {arriveTangent=-1, interpMode=RCIM_Cubic, leaveTangent=-1, time=0.7021, value=0.7}
           {arriveTangent=-4.2574, interpMode=RCIM_Cubic, leaveTangent=-4.2574, tangentMode=RCTM_User, time=1, value=0}
          ]}}})
      Vector2D=<1 1>
     )
    )
    Write to Module Namespace=true
   )
   &Drag(
    Drag=&FloatFromCurve(FloatCurve=DataInterface() {propertyValues {Curve {keys [<at 0 0>,<at 1 5>]}}})
    Use Linear Drag=true
    Use Rotational Drag=true
   )
   &SolveForcesAndVelocity(Clamp Velocity=false, Limit Acceleration=false)
   &DynamicMaterialParameters(
    Index 0 Param 1=&FloatFromCurve(FloatCurve=DataInterface() {propertyValues {Curve {keys [<at 0.2 0>,<at 1 1>]}}})
    Module.Index0Float0Write=true
    Module.Index0Float1Write=true
    Module.Index0Float2Write=true
    Module.Index0Float3Write=true
   )
  ]
  rendererClasses [&NiagaraSpriteRendererProperties]
  renderers [
   {rendererClass=&NiagaraSpriteRendererProperties}
   {rendererClass=&NiagaraSpriteRendererProperties, rendererIndex=1}
  ]
 }
 black001 {
  emitterUpdateScript [
   &EmitterState(Inactive Response=::NewEnumerator1, Life Cycle Mode=::NewEnumerator1, Loop Behavior=::NewEnumerator1, Loop Duration=1, UseLoopDelay=false)
   &SpawnBurst_Instantaneous(Spawn Count=10, Use Spawn Probability=false, UseLoopCountLimit=false)
   &SpawnRate(
    Spawn Probability=&FloatFromCurve(FloatCurve=DataInterface() {propertyValues {Curve {keys [<at 0 0>,<at 0.9 0>,<at 1 1>,<at 8 1>]}}}, CurveIndex=$Emitter.Age)
    SpawnRate=30
    Use Spawn Probability=true
   ) {enabled=false}
  ]
  particleSpawnScript [
   &InitializeParticle(
    Color Mode=::NewEnumerator0
    Lifetime Max=0.3
    Lifetime Min=0.2
    Lifetime Mode=::NewEnumerator1
    Position Offset=<0 0 183>
    Sprite Rotation Mode=::NewEnumerator2
    Sprite Size Max=<20 120>
    Sprite Size Min=<10 80>
    Sprite Size Mode=::NewEnumerator2
    UsePositionOffset=true
    Write Material Random=true
   )
   &ShapeLocation(MaskSpawnGroup=false, Override Seed=false, Sphere Radius=20, Sphere Surface Distribution=1, UseOwnerScale=true)
   &AddVelocity(Module.Use Falloff Distance=false, Origin Offset=<0 0 183>, Override Seed=false, Velocity Mode=::NewEnumerator1, Velocity Speed=260)
  ]
  particleUpdateScript [
   &ParticleState()
   &ScaleSpriteSize(
    Non-Uniform Curve Scale=<1.5 2>
    Non-Uniform Curve Sprite Scale=DataInterface() {propertyValues {
      XCurve {keys [{tangentMode=RCTM_User, time=0, value=1}, {tangentMode=RCTM_User, time=1, value=1}]}
      YCurve {keys [{tangentMode=RCTM_User, time=0, value=0}, <at 0.2 0.8>, {tangentMode=RCTM_User, time=1, value=1}]}
     }}
    Scale Sprite Size Mode=::NewEnumerator3
   )
   &ScaleColor(
    Scale Alpha=&FloatFromCurve(FloatCurve=DataInterface() {propertyValues {Curve {keys [<at 0 1.5>,<at 0.2 1.5>,<at 0.3 0.7>,<at 1 0>]}}}, Scale Curve=1.7)
    Scale RGB=&MakeVectorFromLinearColorRGB(Color=<rgba 0 0 0 1>)
    ScaleA=true
    ScaleRGB=true
   )
   &Drag(Drag=5, Use Linear Drag=true, Use Rotational Drag=true)
   &SolveForcesAndVelocity(Clamp Velocity=false, Limit Acceleration=false)
  ]
  rendererClasses [&NiagaraSpriteRendererProperties]
  renderers [{rendererClass=&NiagaraSpriteRendererProperties}]
 }
 ping {
  emitterUpdateScript [
   &EmitterState(Inactive Response=::NewEnumerator1, Life Cycle Mode=::NewEnumerator1, Loop Behavior=::NewEnumerator1, Loop Duration=2, UseLoopDelay=true)
   &SpawnBurst_Instantaneous(Use Spawn Probability=false, UseLoopCountLimit=false)
  ]
  particleSpawnScript [&InitializeParticle(
    Color Mode=::NewEnumerator0
    Lifetime=0.4
    Sprite Size Mode=::NewEnumerator3
    Uniform Sprite Size=110
    UsePositionOffset=false
    Write Material Random=true
   )]
  particleUpdateScript [
   &ParticleState()
   &ScaleSpriteSize(
    Uniform Curve Scale=1.5
    Uniform Curve Sprite Scale=DataInterface() {propertyValues {Curve {keys [
        {arriveTangent=0.7997, interpMode=RCIM_Cubic, leaveTangent=0.7997, tangentMode=RCTM_User, time=0, value=0.9}
        {arriveTangent=0.5, interpMode=RCIM_Cubic, leaveTangent=0.5, time=0.3, value=1.15}
        {arriveTangent=0.1863, interpMode=RCIM_Cubic, leaveTangent=0.1863, tangentMode=RCTM_User, time=1, value=1.4}
       ]}}}
   )
   &ShapeLocation(MaskSpawnGroup=false, Offset=<0 0 180>, Offset Mode=::NewEnumerator0, Override Seed=false, Sphere Radius=0.1, UseOwnerScale=true)
   &ScaleColor(
    Scale Alpha=&FloatFromCurve(
     FloatCurve=DataInterface() {propertyValues {Curve {keys [
         <at 0 1 RCIM_Cubic>
         {arriveTangent=-1.626, interpMode=RCIM_Cubic, leaveTangent=-1.626, tangentMode=RCTM_User, time=0.6, value=0.6}
         <at 1 0 RCIM_Cubic>
        ]}}}
     Scale Curve=3
    )
    Scale RGB=&MakeVectorFromLinearColorRGB(Color=&ColorFromCurve(ColorCurve=DataInterface() {propertyValues {
        AlphaCurve {keys []}
        BlueCurve {keys [<at 0.601 2 RCIM_Cubic>,<at 0.7211 0.4792 RCIM_Cubic>]}
        GreenCurve {keys [<at 0.601 2 RCIM_Cubic>,<at 0.7211 0.0002 RCIM_Cubic>]}
        RedCurve {keys [<at 0.601 2 RCIM_Cubic>,<at 0.7211 0.285 RCIM_Cubic>]}
       }}))
    ScaleA=true
    ScaleRGB=true
   )
   &FN_Sprite_Attributes(
    Color=&MakeLinearColorFromVectorAndFloat(
     "Vector (RGB)"=&Multiply_VectorByFloat(
      Vector=&VectorFromFloat(Value=&FNTimeOfDayScalar(Night Scale=0.6, Morning Scale=0.25, Evening Scale=0.5))
      Float=&FloatFromCurve(FloatCurve=DataInterface() {propertyValues {Curve {keys [<at 0 1>,<at 0.5 1>,<at 1 0>]}}}, Scale Curve=0.8)
     )
     "Float (Alpha)"=1
    )
    Create Color Attribute=true
    Create Position Attribute=true
    Create Sprite Rotation Attribute=true
    Create SpriteSize Attribute=true
    Module Output:ENiagara_AttributeSamplingApplyOutput=::NewEnumerator1
    Position:NiagaraPosition=<0 0 185>
    SpriteRotation:NiagaraFloat=&RandomRangeFloat(Minimum=0, Maximum=360)
    SpriteSize=&Vector2DFromFloat(Value=&Multiply_Float(
      A=220
      B=&FloatFromCurve(
       FloatCurve=DataInterface() {propertyValues {Curve {keys [
           {tangentMode=RCTM_User, time=0, value=0}
           {arriveTangent=-1.9283, interpMode=RCIM_Cubic, leaveTangent=-1.9283, tangentMode=RCTM_User, time=0.15, value=1}
           {arriveTangent=-0.3812, interpMode=RCIM_Cubic, leaveTangent=-0.01, tangentMode=RCTM_Break, time=0.8, value=0}
          ]}}}
       Scale Curve=3.2
      )
     ))
   )
  ]
  rendererClasses [&NiagaraSpriteRendererProperties]
  renderers [
   {rendererClass=&NiagaraSpriteRendererProperties}
   {rendererClass=&NiagaraSpriteRendererProperties, rendererIndex=1}
  ]
 }
}
summary {
 systemName=NS_Emote_PhoenixKick_EndBig
 emitters [
  {emitterName=RGB, rendererClasses [&NiagaraSpriteRendererProperties]}
  {emitterName=Burst_glow, rendererClasses [&NiagaraSpriteRendererProperties]}
  {emitterName=ping, rendererClasses [&NiagaraSpriteRendererProperties]}
  {emitterName=RGB001, rendererClasses [&NiagaraSpriteRendererProperties]}
  {emitterName=Flare, rendererClasses [&NiagaraSpriteRendererProperties]}
  {emitterName=DiscoBallRaySharp001, simTarget=GPUComputeSim, rendererClasses [&NiagaraMeshRendererProperties]}
  {emitterName=Spikes, rendererClasses [&NiagaraSpriteRendererProperties]}
  {emitterName=black001, rendererClasses [&NiagaraSpriteRendererProperties]}
 ]
 userVariables [
  User.bIsFrontEnd {type=NiagaraBool}
  User.bIsFrontEndPreview {type=NiagaraBool}
 ]
}
